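# Material function for the DigSite set, named for a dance club, that produces an animated offset combining a rocking rotation with a vertical bob, suited to World Position Offset.
# A shared time-times-frequency phase drives a sine angle into RotateAboutAxis, using a parameterized local axis and pivot transformed to world space.
# A second, phase-shifted sine forms a Z-only translation that is added to the rotation offset.
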MaterialFunction /Game/Environments/Sets/DigSite/Materials/MF_Sneezy_DanceClub
graph {nodes [
  n1 = Normalize(VectorInput=TransformPosition(None=VectorParameter() {parameterName=PivotAxis, defaultValue=<rgba 0 1 0 1>}.RGB) {transformType=TRANSFORMPOSSOURCE_World})
  n2 = Multiply(A=Time(), B=ScalarParameter() {defaultValue=2, parameterName=frequency})
  n3 = Multiply(A=Sine(None=@n2), B=ScalarParameter() {defaultValue=0.002, parameterName=RotationAmount})
  n4 = TransformPosition(None=VectorParameter() {parameterName=LocalPivot, defaultValue=<rgba 500 0 -200 1>}.RGB) {transformType=TRANSFORMPOSSOURCE_World}
  n5 = WorldPosition()
  n6 = Constant2Vector()
  n7 = ScalarParameter() {defaultValue=0.2, parameterName=Param_1}
  n8 = Multiply(A=Sine(None=Add(A=@n2, B=@n7)), B=ScalarParameter() {defaultValue=10})
  FunctionOutput(None=Add(A=RotateAboutAxis(NormalizedRotationAxis=@n1, RotationAngle=@n3, PivotPoint=@n4, Position=@n5.XYZ), B=AppendVector(A=@n6, B=@n8)))
 ]}
